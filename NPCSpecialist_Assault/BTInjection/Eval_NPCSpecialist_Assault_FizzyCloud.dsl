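; Behavior tree injection evaluator that decides whether the Assault NPC Specialist can start its FizzyCloud behavior.
; On enter, it caches the controlled pawn, its convert component and the converter pawn.
; Start conditions require the parent check, an enabled scalable-float toggle, the Converted tag and no swimming.
; They pass only when the converter pawn's normalized shield is at or below a scalable-float threshold.
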(event EventOnEnter
  (params (OwnerComp :type Behavior%20Tree%20Component%20Object%20Reference))
  (bind _ascontroller (Utilities|Casting|CastToController (Components|GetOwner OwnerComp)))
  (bind _asfort_pawn (Utilities|Casting|CastToFortPawn (Pawn|GetControlledPawn _ascontroller)))
  (bind _output_get (Variables|Default|SetPawn _asfort_pawn))
  (bind _output_get_2 (Variables|Default|SetConvertComponent (Actor|GetComponentbyClass _output_get "/Script/FortniteGame.FortPawnComponent_Convert")))
  (Variables|Default|SetConverterPawn (Convert|GetConverterPawn _output_get_2)))

(fn EvaluateStartingConditions
  (params (OwnerComp :type Behavior%20Tree%20Component%20Object%20Reference))
  (bind _returnvalue (|CallParentFunction OwnerComp))
  (if (and _returnvalue (ScalableFloat|GetValueasBool (Variables|Default|GetIsEnabled_Hf)))
    (if (or (Default|IsSwimming) (not (GameplayTags|HasMatchingGameplayTag (Variables|Default|GetPawn) "(TagName=\"Athena.Gameplay.Status.Converted\")")))
      (return false)
      (else
        (return (<= (Math|Float|NormalizetoRange (FortPawn|GetShield (Variables|Default|GetConverterPawn)) 0.0 (FortPawn|GetMaxShield (Variables|Default|GetConverterPawn))) (ScalableFloat|GetValueatLevel (Variables|Default|GetShieldPercentThreshold_Hf))))))
    (else
      (return false))))

(fn IsSwimming
  (return (or (WaterBody|IsSurfaceSwimming (Variables|Default|GetPawn)) (AI|Components|NavMovement|IsSwimming (Variables|/Script/Engine.Character|GetCharacterMovement (Variables|Default|GetPawn)) :!class /Script/Engine.NavMovementComponent))))
